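# M_DarkWave is an unlit, two-sided translucent Niagara sprite and mesh particle material for a radial wave effect, colored by particle color through an emissive adjustment function.
# Radial UVs with power-curved V sample two noise textures, whose sum distorts both a plasma texture and a radial edge mask.
# The mask, refined plasma, particle alpha and depth fade multiply together.
# A dynamic parameter overrides panning time, letting particles drive the motion.
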Material /SpireSharedAssets/VFX/Materials/M_DarkWave
graph {
 properties {
  bAutomaticallySetUsageInEditor=false
  bUsedWithNiagaraMeshParticles=true
  bUsedWithNiagaraSprites=true
  blendMode=BLEND_Translucent
  shadingModel=MSM_Unlit
  twoSided=true
  usedShadingModels=Unlit
 }
 nodes [
  n1 = ParticleColor()
  MaterialFunctionCall(Emissive=@n1.RGB) {materialFunction=&MF_FX_EmissiveAdjustments}
  n2 = MaterialFunctionCall(Float2=MaterialFunctionCall() {materialFunction=&VectorToRadialValue}.Radial Coordinates) {materialFunction=&BreakOutFloat2Components}
  n3 = TextureObjectParameter() {group=NoiseA, parameterName=NoiseA_Texture, sortPriority=0, texture=/Game/Items/Traps/Textures/T_PerlinRGB_01_M}
  n4 = VectorParameter() {group=NoiseA, parameterName=NoiseA_UVControl, channelNames {a=VPan, b=UPan, g=VScale, r=UScale}, defaultValue=<rgba 1 0.5 0 -0.7>}
  n5 = DynamicParameter() {paramNames [Time,Param2,Param3,Param4]}
  n6 = MaterialFunctionCall(X=@n2.R, Y=Power(Base=@n2.G, Exp=ScalarParameter() {defaultValue=0.5, parameterName=VPower})) {materialFunction=&MakeFloat2}
  n7 = VectorParameter() {
   group=NoiseA
   parameterName=NoiseA_NoiseControl
   sortPriority=60
   channelNames {a=VStrength, b=UStrength, g=VOffset, r=UOffset}
   defaultValue=<rgba -0.5 -0.5 0.1 0.01>
  }
  n8 = TextureObjectParameter() {
   group=NoiseB
   parameterName=NoiseB_Texture
   samplerType=SAMPLERTYPE_LinearGrayscale
   sortPriority=0
   texture=/Game/Effects/Fort_Effects/Emotes/UltraLight/Texture/T_Tiling_Noise_01_GreyScale
  }
  n9 = VectorParameter() {group=NoiseB, parameterName=NoiseB_UVControl, channelNames {a=VPan, b=UPan, g=VScale, r=UScale}, defaultValue=<rgba 3 0.5 0 -0.5>}
  n10 = VectorParameter() {
   group=NoiseB
   parameterName=NoiseB_NoiseControl
   sortPriority=60
   channelNames {a=VStrength, b=UStrength, g=VOffset, r=UOffset}
   defaultValue=<rgba -0.5 -0.5 0.1 0>
  }
  n11 = Add(
   A=MaterialFunctionCall!(
    NoiseValue=MaterialFunctionCall(Texture=@n3, UVControl_RGB=@n4, UVControl_A=@n4, "[Override]TimeInput"=@n5.Time, "[Override]UVInput"=@n6.Result) {materialFunction=&FXCommon_TextureSample}.RG
    Control_RGB=@n7
    Control_A=@n7
   ) {materialFunction=/Game/Effects/Niagara/MasterMaterials/Functions/Common/FXCommon_NoiseControl}.NoiseValue
   B=MaterialFunctionCall(
    NoiseValue=MaterialFunctionCall(Texture=@n8, UVControl_RGB=@n9, UVControl_A=@n9, "[Override]TimeInput"=@n5.Time, "[Override]UVInput"=@n6.Result) {materialFunction=&FXCommon_TextureSample}.RG
    Control_RGB=@n10
    Control_A=@n10
   ) {materialFunction=&FXCommon_NoiseControl}.NoiseValue
  )
  n12 = VectorParameter() {
   group=Radial
   parameterName=Radial_EdgeMask_InvertControl
   channelNames {a="", b=Power, g=Multiply, r=Offset}
   defaultValue=<rgba 0 1.5 1.4 1>
  }
  n13 = VectorParameter() {group=Radial, parameterName=Radial_EdgeMask_Control, channelNames {a="", b=Power, g=Multiply, r=Offset}, defaultValue=<rgba 0 5 0.5 1>}
  n14 = MaterialFunctionCall!(UorV=Add(A=@n2.G, B=@n11), Invert_EdgeControl=@n12, NoInvert_EdgeControl=@n13) {materialFunction=/Game/Effects/Niagara/MasterMaterials/Functions/Common/FXCommon_EdgeMask_Ex}
  n15 = TextureObjectParameter() {
   group=MainTexture
   parameterName=MainTexture_Texture
   samplerType=SAMPLERTYPE_LinearGrayscale
   sortPriority=0
   texture=/Game/Effects/Fort_Effects/Emotes/UltraLight/Texture/T_Plasma_A_B_Greyscale
  }
  n16 = VectorParameter() {
   group=MainTexture
   parameterName=MainTexture_UVControl
   channelNames {a=VPan, b=UPan, g=VScale, r=UScale}
   defaultValue=<rgba 1 0.2464 0 -0.7>
  }
  n17 = VectorParameter() {parameterName=MainTexture_Refine, defaultValue=<rgba 0.4 1 1 0.5>}
  n18 = MaterialFunctionCall(
   Texture=MaterialFunctionCall(Texture=@n15, AdditionalUVOffset=@n11, UVControl_RGB=@n16, UVControl_A=@n16, "[Override]TimeInput"=@n5.Time, "[Override]UVInput"=@n6.Result) {materialFunction=&FXCommon_TextureSample}.R
   RGB=@n17.RGB
   A=@n17.RGB
  ) {materialFunction=&MF_FX_Refine}
  n19 = ScalarParameter() {defaultValue=300, parameterName=DFDistance}
  Multiply(
   A=Multiply(A=Multiply(A=@n14.Mask, B=@n18.Result), B=@n1.A)
   B=Power(Base=DepthFade(FadeDistance=@n19), Exp=ScalarParameter() {defaultValue=2, parameterName=DFpower})
  )
 ]
}
