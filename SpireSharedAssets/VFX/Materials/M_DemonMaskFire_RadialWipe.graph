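# Unlit translucent Niagara sprite and mesh particle material for demon mask fire with a radial wipe.
# Distorted, panning noise samples in radial or linear UVs (UseRadialUV) blend Details and Highlights colors and build a smoothstep-eroded mask.
# A gradient offset by MaskReveal wipes that mask, which also fades by view angle; dynamic parameters also drive CoreSpeed and DistortionStr.
# Panning samples take explicit derivatives from raw UVs, independent of the radial remap.
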Material /SpireSharedAssets/VFX/Materials/M_DemonMaskFire_RadialWipe
graph {
 properties {
  bUsedWithNiagaraMeshParticles=true
  bUsedWithNiagaraSprites=true
  blendMode=BLEND_Translucent
  pixelDepthOffsetMode=PDOM_Legacy
  shadingModel=MSM_Unlit
  usedShadingModels=Unlit
 }
 nodes [
  LinearInterpolate()
  n1 = TextureCoordinate()
  n2 = Subtract(A=@n1) {constB=0.5}
  n3 = Normalize(VectorInput=@n2)
  n4 = Multiply(
   A=Multiply(A=TextureSample(UVs=@n3) {samplerType=SAMPLERTYPE_Masks, texture=/SpireSharedAssets/VFX/Textures/T_DemonMaskFire_Radial}.R, B=@n3)
   B=DynamicParameter() {paramNames [CoreSpeed, MaskReveal, DistortionStr, Param4]}.DistortionStr
  )
  n5 = Multiply(A=NamedRerouteUsage(), B=ComponentMask(None=VectorParameter() {parameterName=CoreUVScale, defaultValue=<rgba 3 3 0 1>}.RGB) {g=true, r=true})
  n6 = DynamicParameter() {paramNames [CoreSpeed, MaskReveal, DistortionStr, Param4]}
  n7 = AppendVector(A=Multiply(B=NamedRerouteUsage()) {constA=0.25}, B=Constant() {r=-1})
  n8 = TextureCoordinate()
  n9 = DDX(Value=@n8)
  n10 = DDY(Value=@n8)
  n11 = TextureSample!(
   UVs=Add(A=@n4, B=Panner(Coordinate=@n5, Time=@n6.CoreSpeed, Speed=@n7) {speedY=-1})
   Tex
   "DDX(UVs)"=@n9
   "DDY(UVs)"=@n10
   Apply View MipBias
  ) {mipValueMode=TMVM_Derivative, samplerType=SAMPLERTYPE_Masks, texture=&T_DemonMaskFire_Radial}
  LinearInterpolate(
   A=VectorParameter() {parameterName=Details, defaultValue=<rgba 0 0 0 1>}.RGB
   B=VectorParameter() {parameterName=Highlights, defaultValue=<rgba 0 0 0 1>}.RGB
   Alpha=@n11.B
  )
  n12 = Multiply(A=@n5, B=Constant2Vector() {g=1, r=3}) {constB=3}
  n13 = TextureSample(UVs=Add(A=@n4, B=Panner(Coordinate=@n12) {speedY=-1.5}), "DDX(UVs)"=@n9, "DDY(UVs)"=@n10) {mipValueMode=TMVM_Derivative, samplerType=SAMPLERTYPE_Masks, texture=&T_DemonMaskFire_Radial}
  n14 = Multiply(A=OneMinus(None=NamedRerouteUsage()), B=ScalarParameter() {defaultValue=2, parameterName=MaskScale}) {constB=2}
  n15 = DynamicParameter() {paramNames [CoreSpeed, MaskReveal, DistortionStr, Param4]}
  n16 = Multiply(
   A=Add(A=@n11.B, B=@n13.G)
   B=TextureSample!(UVs=Add(A=@n14, B=@n15.MaskReveal), Tex, Apply View MipBias) {samplerSource=SSM_Clamp_WorldGroupSettings, texture=/Game/Effects/Fort_Effects/UI/Transition_IAS/Resources/T_Vert_Gradient_Soft}.R
  )
  n17 = TextureCoordinate()
  n18 = Constant() {r=0.5}
  n19 = ComponentMask(None=TextureCoordinate()) {g=true}
  n20 = Add(
   A=StaticSwitchParameter(True=SphereMask(A=@n17, B=@n18) {attenuationRadius=0.5, hardnessPercent=70}, False=@n19) {defaultValue=true, parameterName=UseRadialUV}
   B=ScalarParameter() {parameterName=p}
  )
  n21 = SmoothStep(Min=Add(A=@n20, B=ScalarParameter() {parameterName=w}), Max=@n20, Value=@n13.R)
  n22 = Constant() {r=1}
  n23 = OneMinus(None=@n19)
  Multiply(
   A=Multiply(A=Multiply(A=@n16, B=@n21), B=StaticSwitchParameter(True=@n22, False=@n23) {defaultValue=true, parameterName=UseRadialUV})
   B=MaterialFunctionCall("FadeSensitivity (optional)"=ScalarParameter() {parameterName="FadeSensitivity (optional) (S)"}) {materialFunction=&FadeBasedOnViewAngle}.Result
  )
  n24 = MaterialFunctionCall(Vector or UVs=@n2) {materialFunction=&VectorToRadialValue}
  NamedRerouteDeclaration(None=StaticSwitchParameter(True=@n24.Radial Coordinates, False=@n1) {defaultValue=true, parameterName=UseRadialUV}) {name=RadialUVs, nodeColor=<rgba 0 0.7187 1 1>}
  NamedRerouteDeclaration(None=StaticSwitchParameter(True=@n24.Linear Distance, False=ComponentMask(None=@n1) {g=true}) {defaultValue=true, parameterName=UseRadialUV}) {name=RadialDistance, nodeColor=<rgba 0.7578 1 0 1>}
  ParticleColor()
  SphereMask(A=@n19, B=Constant() {r=1}) {attenuationRadius=1, hardnessPercent=0}
  StaticSwitchParameter() {defaultValue=true, parameterName=UseRadialUV}
  TextureCoordinate()
 ]
}
