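# Translucent, unlit, two-sided Niagara sprite and mesh-particle material for electric lightning effects, named for a slag destructible.
# A bolt texture panned by particle time, randomly offset and smoke-distorted, is multiplied by tip, base and width masks into an SDF-style value.
# Glow and core falloffs are screen-blended into opacity, with depth fade skipped on lower quality.
# Dynamic parameters drive length, pan and UV inversion; WPO combines spline thickening with a camera-ward offset.
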Material /SpireSharedAssets/VFX/Materials/M_FX_SlagDestructible_Elec
graph {
 properties {
  bAutomaticallySetUsageInEditor=false
  bUsedWithNiagaraMeshParticles=true
  bUsedWithNiagaraSprites=true
  blendMode=BLEND_Translucent
  shadingModel=MSM_Unlit
  twoSided=true
  usedShadingModels=Unlit
 }
 nodes [
  AppendVector(A=Constant(), B=ParticleRandom())
  n1 = ParticleColor()
  MaterialFunctionCall(Emissive=@n1.RGB) {materialFunction=&MF_FX_EmissiveAdjustments}
  n2 = Constant() {r=1}
  n3 = QualitySwitch(Default=DepthFade(FadeDistance=ScalarParameter() {defaultValue=3, parameterName=Depth Fade Distance}), Low=@n2, Medium=@n2)
  n4 = Multiply(
   A=Power(Base=NamedRerouteUsage(), Exp=ScalarParameter() {defaultValue=2, parameterName=Lightning Glow Exponent})
   B=ScalarParameter() {defaultValue=0.1, parameterName=Lightning Glow Intensity}
  )
  n5 = OneMinus(None=NamedRerouteUsage())
  n6 = Saturate(None=ScalarParameter() {defaultValue=0.2, parameterName=Lightning Core Size})
  n7 = ScalarParameter() {defaultValue=0.8, parameterName=Lightning Core Softness}
  n8 = Divide(A=Saturate(None=Subtract(A=@n5, B=@n6)), B=Multiply(A=OneMinus(None=@n6), B=Saturate(None=@n7)))
  n9 = Power(Base=OneMinus(None=Saturate(None=@n8)), Exp=ScalarParameter() {defaultValue=5, parameterName=Lightning Core Exponent})
  n10 = OneMinus(None=Multiply(A=OneMinus(None=@n4), B=OneMinus(None=@n9)))
  Multiply(A=@n1.A, B=NamedRerouteDeclaration(None=Multiply(A=@n3, B=@n10)) {name=Opacity, nodeColor=<rgba 1 0 0.9375 1>})
  Multiply()
  n11 = Multiply(A=NamedRerouteUsage(), B=NamedRerouteUsage())
  n12 = NamedRerouteUsage()
  n13 = ScalarParameter() {parameterName=ScaleY}
  n14 = DynamicParameter() {defaultValue=<rgba 10 0 0 0>, paramNames [Lightning Width, Lightning Length, Lightning Pan, Lightning Invert]}
  n15 = MaterialFunctionCall(UVs=TextureCoordinate(), Texture Scale=AppendVector(A=ScalarParameter() {parameterName=ScaleX}, B=Divide(A=@n13, B=@n14.Lightning Length))) {materialFunction=&ScaleUVsByCenter}
  n16 = NamedRerouteUsage()
  n17 = Panner(Coordinate=@n15.UVs, Time=@n16) {bFractionalPart=true, speedY=2}
  n18 = Multiply(A=Add(A=ParticleRandom()), B=Constant2Vector() {g=1})
  n19 = ScalarParameter() {parameterName=Distortion}
  n20 = TextureSample(UVs=Panner(Coordinate=@n15.UVs, Time=@n16) {bFractionalPart=true, speedX=1.5, speedY=3}) {texture=/Game/Effects/Fort_Effects/GameplayPulginsAssets/Doom/DoomCastle/Smoke/MaterialAndMesh/Smoke01_D}
  n21 = TextureSample(UVs=Add(A=Add(A=@n17, B=@n18), B=Multiply(A=@n19, B=@n20.R))) {
   samplerType=SAMPLERTYPE_LinearGrayscale
   texture=/BRCosmetics/Athena/Effects/Particles/Skydiving/NS_Contrail_GreatPool/Meshes/T_FX_GreatPool_Bolts
  }
  NamedRerouteDeclaration(None=Multiply(A=Multiply(A=@n11, B=@n12), B=@n21.R) {constA=1}) {name=LightningSDF, nodeColor=<rgba 0.8047 1 0 1>}
  n22 = ComponentMask(None=NamedRerouteUsage()) {g=true}
  n23 = Subtract(A=@n22, B=@n14.Lightning Pan)
  n24 = ScalarParameter() {defaultValue=0.1, parameterName=Opacity - Fade Size Tip}
  NamedRerouteDeclaration(None=Saturate(None=Divide(A=@n23, B=@n24))) {name=Length Mask, nodeColor=<rgba 1 0 0.8438 1>}
  n25 = TextureCoordinate()
  n26 = MaterialFunctionCall(Float2=@n25) {materialFunction=&BreakOutFloat2Components}
  n27 = OneMinus(None=@n26.G)
  n28 = DynamicParameter() {defaultValue=<rgba 10 0 0 0>, paramNames [Lightning Width, Lightning Length, Lightning Pan, Lightning Invert]}
  n29 = Saturate(None=@n28.Lightning Invert)
  NamedRerouteDeclaration(None=LinearInterpolate(
    A=@n25
    B=MaterialFunctionCall(X=@n26.R, Y=@n27) {materialFunction=&MakeFloat2}.Result
    Alpha=Round(None=@n29)
   ) {constAlpha=1}) {name=UVs, nodeColor=<rgba 1 0.5859 0 1>}
  n30 = ComponentMask(None=TextureCoordinate()) {r=true}
  n31 = OneMinus(None=Abs(None=ConstantBiasScale(None=@n30) {bias=-0.5, scale=2}))
  n32 = ScalarParameter() {defaultValue=0.2, parameterName=Opacity - Fade Size Width}
  NamedRerouteDeclaration(None=Saturate(None=Divide(A=@n31, B=@n32))) {name=Width Mask, nodeColor=<rgba 0 0.2969 1 1>}
  n33 = WorldPosition()
  n34 = NamedRerouteUsage()
  n35 = Normalize(VectorInput=Subtract(A=CameraPositionWS(), B=Add(A=@n33.XYZ, B=@n34)))
  n36 = ScalarParameter() {defaultValue=22, parameterName=Camera Offset}
  NamedRerouteDeclaration(None=Add(A=Multiply(A=@n35, B=@n36), B=NamedRerouteUsage())) {name=WPO, nodeColor=<rgba 1 0 0.5391 1>}
  n37 = DynamicParameter() {parameterIndex=1, defaultValue=<rgba 10 0 0 0>, paramNames [Lightning Width, Lightning Length, Lightning Pan, Lightning Invert]}
  NamedRerouteDeclaration(None=@n37.Lightning Pan) {name=ParticleTime, nodeColor=<rgba 0 1 0.5078 1>}
  n38 = OneMinus(None=@n22)
  n39 = ScalarParameter() {defaultValue=0.1, parameterName=Opacity - Fade Size Base}
  NamedRerouteDeclaration(None=Saturate(None=Divide(A=@n38, B=@n39))) {name=Base Mask, nodeColor=<rgba 1 0.9375 0 1>}
  n40 = ScalarParameter() {parameterName=Thickness}
  n41 = TextureCoordinate()
  n42 = LinearInterpolate(
   A=Multiply(A=@n40, B=ScalarParameter() {parameterName=EndThickness})
   B=@n40
   Alpha=SphereMask(A=ComponentMask(None=@n41) {g=true}, B=Constant() {r=0.5}) {attenuationRadius=0.5, hardnessPercent=20}
  )
  NamedRerouteDeclaration(None=MaterialFunctionCall(WidthBase=@n42, WidthTip=@n42) {materialFunction=&SplineThicken}.WorldPositionOffset) {name=SplineThicken WPO, nodeColor=<rgba 0.2891 0 1 1>}
  NamedRerouteUsage()
  Reroute()
 ]
}
