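# Additive unlit lens flare material for Niagara sprites in Spire, tinted by particle color.
# A panning animation texture slightly warps the base flare's V coordinate, and the flare's alpha adds a tinted layer scaled by the AnimFlare dynamic parameter.
# Offset per-channel gamma separates red and blue like chromatic aberration, then HSV offsets and brightness shape the emissive.
# A separate particle-alpha-weighted branch blends raw flare and gamma-shifted color by Center Falloff.
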Material /SpireSharedAssets/VFX/Materials/M_LensFlare_Spire
graph {
 properties {bUsedWithNiagaraSprites=true, blendMode=BLEND_Additive, shadingModel=MSM_Unlit, usedShadingModels=Unlit}
 nodes [
  n1 = ParticleColor()
  n2 = DynamicParameter() {defaultValue=<rgba 1 0 1 1>, paramNames [Param1, AnimFlare, Param3, Center Falloff]}
  n3 = Constant()
  n4 = MaterialFunctionCall(UVs=TextureCoordinate() {vTiling=2.5}, Texture Scale=Constant2Vector() {g=0.25, r=2.5}) {materialFunction=&ScaleUVsByCenter}
  n5 = TextureSampleParameter2D(UVs=Panner(Coordinate=MaterialFunctionCall(XY=@n4.UVs) {materialFunction=&Swizzle}.YX) {speedX=1, speedY=2}) {group=BaseFlare, parameterName=Param, samplerType=SAMPLERTYPE_LinearColor, texture=/SpireSharedAssets/VFX/Textures/T_LensFlare_Anim}
  n6 = Multiply(A=@n5.G) {constB=0.0075}
  n7 = TextureSampleParameter2D(UVs=Add(A=AppendVector(A=@n3, B=@n6), B=TextureCoordinate())) {
   group=BaseFlare
   parameterName=BaseFlare
   samplerType=SAMPLERTYPE_LinearColor
   texture=/Game/Effects/Fort_Effects/GameplayPulginsAssets/SPire/T_Lens_Flare
  }
  n8 = Multiply(A=@n2.AnimFlare, B=@n7.A)
  n9 = Multiply(A=@n7.RGB, B=@n1.RGB)
  n10 = Clamp(None=@n9)
  n11 = MaterialFunctionCall(Float3=Add(
    A=Multiply(A=@n1.RGB, B=@n8)
    B=MaterialFunctionCall(LDR Input=@n10, HDR Tint=@n9) {materialFunction=&DeriveHDRfromLDR}.RemappedRange
   )) {materialFunction=&BreakOutFloat3Components}
  n12 = ScalarParameter() {defaultValue=0.5, parameterName=ChannelGammaAbbe, sliderMax=3}
  n13 = ScalarParameter() {defaultValue=1, parameterName=ChannelGamma, sliderMax=3}
  n14 = MaterialFunctionCall(
   X=Power(Base=@n11.R, Exp=Add(A=@n13, B=@n12)) {constExponent=1.5}
   Y=Power(Base=@n11.G, Exp=@n13) {constExponent=1}
   Z=Power(Base=@n11.B, Exp=Subtract(A=@n13, B=@n12)) {constExponent=1.2}
  ) {materialFunction=&MakeFloat3}
  n15 = MaterialFunctionCall(Float3=RgbToHsv!(None=@n14.Result)) {materialFunction=&BreakOutFloat3Components}
  n16 = MaterialFunctionCall(
   X=Add(A=ScalarParameter() {parameterName=Hue Offset, sliderMax=1, sliderMin=-1}, B=@n15.R)
   Y=Add(A=ScalarParameter() {parameterName=Sat Offset, sliderMax=1, sliderMin=-1}, B=@n15.G)
   Z=@n15.B
  ) {materialFunction=&MakeFloat3}
  MaterialFunctionCall(Emissive=Multiply(A=HsvToRgb!(None=@n16.Result), B=ScalarParameter() {defaultValue=1, parameterName=FinalBrightness, sliderMax=1})) {materialFunction=&MF_FX_EmissiveAdjustments}
  n17 = Add(A=@n7.RGB, B=@n7.A)
  Multiply(A=LinearInterpolate(A=Multiply(A=@n1.A, B=@n17), B=@n14.Result, Alpha=OneMinus(None=@n2.Center Falloff)), B=@n1.A)
 ]
}
