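# Additive, unlit, two-sided light pillar material for meshes and Niagara particles, with switches for bunker pillar and train beam variants.
# Panning smoke noise and a color gradient, tinted by particle color or BaseColor, form HDR emissive.
# A textured vertical gradient drives opacity through depth, near-camera and end fades plus optional masks.
# SplineThicken sets beam width in world position offset, with camera-distance scale-up keeping distant pillars large.
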Material /SpireSharedAssets/VFX/Materials/M_LightPillar
graph {
 properties {
  bUsedWithNanite=true
  bUsedWithNiagaraMeshParticles=true
  bUsedWithNiagaraRibbons=true
  bUsedWithNiagaraSprites=true
  blendMode=BLEND_Additive
  pixelDepthOffsetMode=PDOM_Legacy
  shadingModel=MSM_Unlit
  twoSided=true
  usedShadingModels=Unlit
 }
 nodes [
  n1 = ScalarParameter() {defaultValue=0.2, parameterName=Additional Distance Fade Ofs}
  n2 = Subtract(
   A=MaterialFunctionCall(Distance=ScalarParameter() {defaultValue=4000, parameterName=Additional Distance Fade}) {materialFunction=&Near_Camera_Fade}.Result
   B=@n1
  )
  n3 = Saturate(None=Multiply(A=Saturate(None=@n2), B=Add(A=@n1)))
  n4 = Multiply(
   A=TextureCoordinate()
   B=AppendVector(
    A=ScalarParameter() {defaultValue=1, group=SphereMask, parameterName=SphereMask-UV-X}
    B=ScalarParameter() {defaultValue=0.3, group=SphereMask, parameterName=SphereMask-UV-Y}
   )
  )
  n5 = ComponentMask(None=@n4) {g=true}
  n6 = ScalarParameter() {group=SphereMask, parameterName=Spheremask UV.Y Min}
  n7 = ScalarParameter() {defaultValue=0.5, group=SphereMask, parameterName=Spheremask UV.Y Max}
  n8 = StaticSwitchParameter(True=AppendVector(A=ComponentMask(None=@n4) {r=true}, B=Clamp(None=@n5, Min=@n6, Max=@n7) {maxDefault=0.5}), False=@n4) {
   desc=Clamps spheremask uv Y component. Use to make spheremask only affect one portion of the length of the beam
   group=SphereMask
   parameterName=Clamp Spheremask UV.Y
  }
  n9 = SphereMask(
   A=@n8
   B=AppendVector(
    A=ScalarParameter() {defaultValue=0.5, group=SphereMask, parameterName=SphereMask-CenterX}
    B=ScalarParameter() {defaultValue=-0.225, group=SphereMask, parameterName=SphereMask-CenterY}
   )
   Radius=ScalarParameter() {defaultValue=0.55, group=SphereMask, parameterName=SphereMask-Radius}
   Hardness=ScalarParameter() {group=SphereMask, parameterName=SphereMask-Hardness}
  )
  n10 = ParticleColor()
  n11 = VectorParameter() {parameterName=BaseColor, defaultValue=<rgba 1 1 1 1>}
  n12 = StaticSwitchParameter(True=@n10.A, False=@n11.A) {defaultValue=true, parameterName=UseWParticles}
  n13 = MaterialFunctionCall() {materialFunction=&LinearGradient}
  n14 = VectorParameter() {parameterName=MainTexturePanerSpeed, defaultValue=<rgba 0 0 0 1>}
  n15 = TextureSampleParameter2D(UVs=Panner(Coordinate=TextureCoordinate() {coordinateIndex=1}, Speed=ComponentMask(None=@n14.RGB) {g=true, r=true})) {parameterName=MainTexture, samplerType=SAMPLERTYPE_LinearColor, texture=/Game/Global/Textures/Default/Masks_and_Gradients/T_VertGradient}
  n16 = Power(Base=@n15.R)
  n17 = Multiply(
   A=@n16
   B=Power(Base=OneMinus(None=@n13.VGradient), Exp=ScalarParameter() {defaultValue=6, parameterName=GradientFadePower}) {constExponent=6}
  )
  n18 = ScalarParameter() {defaultValue=300, parameterName=DepthFadeDistance}
  n19 = ScalarParameter() {defaultValue=1, parameterName=DepthFade-NearFadeBasedMultiplier}
  n20 = MaterialFunctionCall(Distance=ScalarParameter() {defaultValue=200, parameterName=NearFadeDistance}) {materialFunction=&Near_Camera_Fade}
  n21 = DepthFade(Opacity=@n17, FadeDistance=LinearInterpolate(A=@n18, B=Multiply(A=@n18, B=@n19), Alpha=@n20.Result)) {fadeDistanceDefault=200}
  n22 = Multiply(
   A=Multiply(
    A=StaticSwitchParameter(True=@n10.A, False=@n12) {parameterName=for Bunker Pillar}
    B=ShadingPathSwitch(Default=@n21, Mobile=@n17) {desc=Migrated from FeatureLevelSwitch}
   )
   B=@n20.Result
  )
  n23 = ObjectPositionWS()
  n24 = Distance(
   A=StaticSwitchParameter(True=VectorParameter() {parameterName=BeaconLocation}.RGB, False=@n23) {parameterName=Use Controlled Distance Vectors for Scaling}
   B=CameraPositionWS()
  )
  n25 = ScalarParameter() {defaultValue=2000, parameterName=DistanceOffset}
  n26 = Clamp(None=Divide(A=Subtract(A=@n24, B=@n25), B=ScalarParameter() {defaultValue=30000, parameterName=ScaleUpDistance}))
  n27 = StaticSwitchParameter(True=OneMinus(None=@n26), False=@n26) {parameterName=Use Controlled Distance Vectors for Scaling}
  n28 = Multiply(
   A=ScalarParameter() {defaultValue=1, parameterName=Opacity}
   B=StaticSwitchParameter(True=LinearInterpolate(B=@n22, Alpha=@n27), False=@n22) {parameterName=Use Controlled Distance Vectors for Scaling}
  )
  n29 = MaterialFunctionCall() {materialFunction=&LinearGradient}
  n30 = OneMinus(None=@n29.VGradient)
  n31 = Multiply(
   A=Power(Base=Multiply(A=@n29.VGradient, B=@n30), Exp=ScalarParameter() {defaultValue=1, group=Fade Top Bottom, parameterName=Fade Top Bottom Power})
   B=ScalarParameter() {defaultValue=1, group=Fade Top Bottom, parameterName=Fade Top Bottom Multiplier}
  )
  n32 = LinearInterpolate(A=@n28, B=Multiply(A=@n28, B=Clamp(None=@n31)), Alpha=ScalarParameter() {group=Fade Top Bottom, parameterName=Fade Top Bottom})
  n33 = TextureCoordinate()
  n34 = Multiply(A=Power(Base=ComponentMask(None=@n33) {g=true}, Exp=ScalarParameter() {defaultValue=1, parameterName=LinearMask_Range}) {constExponent=1}) {constB=4}
  n35 = Multiply(
   A=LinearInterpolate(A=@n32, B=Multiply(A=@n32, B=@n9), Alpha=ScalarParameter() {group=SphereMask, parameterName=UseSphereMask, sortPriority=1})
   B=LinearInterpolate(B=Saturate(None=@n34), Alpha=ScalarParameter() {parameterName=UseLinearMask}) {constA=1}
  )
  n36 = ComponentMask(None=TextureCoordinate()) {g=true}
  n37 = Multiply(A=@n36) {constB=40}
  Multiply(
   A=StaticSwitchParameter(True=Multiply(A=@n3, B=@n35), False=@n35) {parameterName=for Bunker Pillar}
   B=LinearInterpolate(B=Saturate(None=@n37), Alpha=ScalarParameter() {parameterName=Use Edge Mask}) {constA=1}
  )
  OneMinus(None=@n36)
  n38 = StaticSwitchParameter(True=@n10.RGB, False=@n11.RGB) {defaultValue=true, parameterName=UseWParticles}
  n39 = DynamicParameter() {defaultValue=<rgba 100 1 1 1>, paramNames [SplineWidth, NoiseOffsetU, NoiseOffsetV, VerticalSpeed]}
  n40 = Multiply(A=AppendVector(A=@n39.NoiseOffsetU, B=@n39.NoiseOffsetU), B=ScalarParameter() {parameterName=UseDynamicNoiseOffset})
  n41 = TextureCoordinate()
  n42 = Constant()
  n43 = DynamicParameter() {defaultValue=<rgba 100 1 1 1>, paramNames [SplineWidth, NoiseOffsetU, NoiseOffsetV, VerticalSpeed]}
  n44 = ComponentMask(None=VectorParameter() {parameterName=NoiseTiling, defaultValue=<rgba 1 5 0 0>}.RGB) {g=true, r=true}
  n45 = ScalarParameter() {defaultValue=-0.05, parameterName=SpeedX}
  n46 = ScalarParameter() {defaultValue=-0.25, parameterName=SpeedY}
  n47 = StaticSwitchParameter(
   True=Add(A=@n41, B=AppendVector(A=@n42, B=@n43.VerticalSpeed))
   False=Panner(Coordinate=Multiply(A=@n41, B=@n44), Speed=AppendVector(A=@n45, B=@n46)) {speedX=-0.05, speedY=-0.25}
  ) {
   desc=Clamps spheremask uv Y component. Use to make spheremask only affect one portion of the length of the beam
   group=SphereMask
   parameterName=AdjustVerticalSpeedByParameter
  }
  n48 = Multiply(
   A=TextureSample(UVs=Add(A=@n40, B=@n47)) {texture=&T_Inky_Smoke_Tile_Chromatic}.RGB
   B=ScalarParameter() {defaultValue=1, parameterName=NoiseOpacity}
  )
  n49 = LinearInterpolate(
   A=Multiply(A=@n16) {constB=-0.5}
   B=Multiply(A=@n16) {constB=0.5}
   Alpha=ComponentMask(None=TextureCoordinate() {uTiling=1.5, vTiling=1.5}) {r=true}
  )
  n50 = LinearInterpolate(
   A=TextureSampleParameter2D(UVs=Panner(Coordinate=@n49) {speedX=0.8, speedY=0.8}) {parameterName=ColorGradient, texture=/Game/Effects/Fort_Effects/Textures/Grad/T_Gradient_Rainbow_NoRed}.RGB
   Alpha=@n16
  )
  n51 = Multiply(A=Desaturation(None=Add(A=@n48, B=@n50), Fraction=ScalarParameter() {defaultValue=-1, parameterName=Desaturation}), B=@n38)
  n52 = MaterialFunctionCall(LDR Input=Clamp(None=@n38), HDR Tint=@n51, HDR Intensity Multiplier=ScalarParameter() {defaultValue=1, parameterName=HDRIntensity}) {materialFunction=&DeriveHDRfromLDR}
  Reroute(None=MaterialFunctionCall(Emissive=@n52.RemappedRange) {materialFunction=&MF_FX_EmissiveAdjustments}.Result)
  n53 = ScalarParameter() {defaultValue=60, parameterName=MaxScaleIncrease}
  n54 = StaticSwitchParameter(
   True=MaterialFunctionCall() {materialFunction=&ComponentPivotLocation}.Result
   False=@n23
  ) {parameterName=IsTrainBeam}
  n55 = WorldPosition()
  n56 = LinearInterpolate(B=Multiply(A=Multiply(A=@n53) {constB=-1}, B=Subtract(A=@n54, B=@n55.XYZ)) {constB=-1}, Alpha=@n27) {constB=0}
  n57 = Constant()
  n58 = ScalarParameter() {group=Camera Distance Spline Scale, parameterName=Spline Scale Camera Distance Min}
  n59 = ScalarParameter() {defaultValue=1, group=Camera Distance Spline Scale, parameterName=Spline Scale Camera Distance Max, sortPriority=33}
  n60 = ScalarParameter() {defaultValue=1, parameterName=SplineWidth}
  n61 = Multiply(
   A=Max(A=MaterialFunctionCall(Input=@n20.Result, Target Low=@n58, Target High=@n59) {materialFunction=&RemapValueRange}.Result) {constB=0}
   B=@n60
  )
  n62 = ScalarParameter() {defaultValue=20, parameterName=MaxWidthMult}
  n63 = Multiply(
   A=StaticSwitchParameter(
    True=DynamicParameter() {defaultValue=<rgba 100 1 1 1>, paramNames [SplineWidth, NoiseOffsetU, NoiseOffsetV, VerticalSpeed]}.SplineWidth
    False=StaticSwitchParameter(True=@n61, False=@n60) {parameterName=Enable Spline Scale With Camera Fade}
   ) {defaultValue=true, parameterName=UseWParticles}
   B=StaticSwitchParameter(True=LinearInterpolate(B=@n62, Alpha=@n27) {constA=1, constB=3}, False=Constant() {r=1}) {defaultValue=true, parameterName=UseScaleUpDistance}
  )
  Reroute(None=Add(
    A=MaterialFunctionCall(WidthBase=@n63, WidthTip=@n63) {materialFunction=&SplineThicken}.WorldPositionOffset
    B=StaticSwitchParameter(True=@n56, False=@n57) {defaultValue=true, parameterName=UseScaleUpDistance}
   ))
  ShadingPathSwitch(Default=@n52.RemappedRange, Mobile=@n51) {desc=Migrated from FeatureLevelSwitch}
 ]
}
