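# Unlit material for Niagara mesh particles, apparently slag destructible debris, reusing rock textures from the ceiling collapse trap.
# A normal-mapped Fresnel rim tint is added to particle color and a fake point light.
# That light is a soft, powered sphere mask around the particle position, with dynamic-parameter radius and brightness, tinted by LightColor and modulated by the diffuse red channel.
# The sum passes through MF_FX_EmissiveAdjustments.
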Material /SpireSharedAssets/VFX/Materials/M_SlagDestructible
graph {
 properties {bAutomaticallySetUsageInEditor=false, bUsedWithNiagaraMeshParticles=true, shadingModel=MSM_Unlit, usedShadingModels=Unlit}
 nodes [
  n1 = TextureSample() {samplerType=SAMPLERTYPE_Normal, texture=/Game/Weapons/FORT_TRAPS/Ceiling_Collapse_Trap/Texture/T_Ceiling_Collapse_Trap_Rocks_N}
  n2 = Multiply(
   A=Fresnel(Normal=Transform(None=@n1.RGB)) {baseReflectFraction=0, exponent=3}
   B=VectorParameter() {parameterName=FresnelColor, defaultValue=<rgba 0 0 0 1>}.RGB
  )
  n3 = ParticleColor()
  n4 = TextureSample() {texture=/Game/Weapons/FORT_TRAPS/Ceiling_Collapse_Trap/Texture/T_Ceiling_Collapse_Trap_Rocks_D}
  n5 = ParticlePositionWS()
  n6 = WorldPosition()
  n7 = DynamicParameter() {paramNames [LightRadius, LightBrightness, Param3, Param4]}
  n8 = Multiply(
   A=Power(Base=SphereMask(A=@n5, B=@n6.XYZ, Radius=@n7.LightRadius) {hardnessPercent=0}) {constExponent=4}
   B=Multiply(A=VectorParameter() {parameterName=LightColor, defaultValue=<rgba 0 0 0 1>}.RGB, B=@n7.LightBrightness)
  )
  MaterialFunctionCall(Emissive=Add(A=Add(A=@n2, B=@n3.RGB), B=Multiply(A=@n4.R, B=@n8))) {materialFunction=&MF_FX_EmissiveAdjustments}
 ]
}
