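# Niagara system for the impact effect of a spire's destruction, built from flash, sparks, soft glow and two smoke emitters (central and smaller).
# It exposes a User.FXScale float for effect scale.
# The system spawn script uses two Set Variables modules, and float-multiply and float-to-Vector2D dynamic inputs are among its dependencies.
# No renderers or data interfaces are listed.
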NiagaraSystem /SpireSharedAssets/VFX/NSLW_SpireDestructionImpact
dependencies {
 usedDataInterfaces []
 usedDynamicInputs [&Vector2DFromFloat, &Multiply_Float]
 usedModules [
  &self:SystemSpawnScript.SystemScriptSource.SystemScriptGraph.NiagaraNodeAssignment_0.SetVariables_#1
  &self:SystemSpawnScript.SystemScriptSource.SystemScriptGraph.NiagaraNodeAssignment_1.SetVariables_#2
  &SystemState
 ]
}
emitters {
 NE_Juno_Flash_LW {rendererClasses [], renderers []}
 NE_Juno_Sparks_LW {rendererClasses [], renderers []}
 SmokeCentral {rendererClasses [], renderers []}
 SmokeSmaller {rendererClasses [], renderers []}
 SoftGlow {rendererClasses [], renderers []}
}
summary {
 systemName=NSLW_SpireDestructionImpact
 emitters [
  {emitterName=NE_Juno_Flash_LW, rendererClasses []}
  {emitterName=NE_Juno_Sparks_LW, rendererClasses []}
  {emitterName=SmokeSmaller, rendererClasses []}
  {emitterName=SoftGlow, rendererClasses []}
  {emitterName=SmokeCentral, rendererClasses []}
 ]
 userVariables [User.FXScale {type=NiagaraFloat, defaultValue=10, description=""}]
}
