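# Niagara system for Mask Maker spire shards: a lingering crimson flare glow with a light, built from a single Flash_Flare emitter.
# A one-particle burst spawns a tall sprite, offset upward with a vertical alignment vector, that does not die when its lifetime elapses.
# The particle drives two sprite renderers and a light renderer, with separate Energy size, color and dynamic-parameter attributes derived alongside the dimmed base color.
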NiagaraSystem /SpireSharedAssets/VFX/NS_MaskMaker_SpireShards
dependencies {
 usedDataInterfaces [&NiagaraDataInterfaceCurve, &NiagaraDataInterfaceSkeletalMesh]
 usedDynamicInputs [&FN_LOD_Distance_Remap]
 usedModules [
  &self:SystemSpawnScript.SystemScriptSource.SystemScriptGraph.NiagaraNodeAssignment_0.SetVariables_#1
  &SystemState
  &self:SystemSpawnScript.SystemScriptSource.SystemScriptGraph.NiagaraNodeAssignment_1.SetVariables_#2
 ]
}
emitters {Flash_Flare {
  emitterUpdateScript [
   &EmitterState(Life Cycle Mode=::NewEnumerator1, Loop Behavior=::NewEnumerator1, Loop Duration Mode=::NewEnumerator1, UseLoopDelay=false)
   &SpawnBurst_Instantaneous(Spawn Count=1, Spawn Time=0.1, Use Spawn Probability=false)
  ]
  particleSpawnScript [
   &InitializeParticle(
    Color=<rgba 0.3142 0.0086 0.067 1>
    Lifetime=1.5
    Sprite Size=<600 1200>
    Sprite Size Mode=::NewEnumerator4
    UsePositionOffset=false
    Write Material Random=true
   )
   &DynamicMaterialParameters(
    Index 0 Param 1=0.4
    Index 0 Param 2=1
    Module.Index0Float0Write=true
    Module.Index0Float1Write=true
    Module.Index0Float2Write=true
    Module.Index0Float3Write=true
   )
   /SetVariables(Particles.ParticleAlignmentVector:Vector3f=<0 0 1>) {isSetParameters=true}
   &SystemLocation(MaskSpawnGroup=false, Offset=<0 0 50>)
   /SetVariables(Particles.EnergySize:Vector2f=&Multiply_Vector2D(A=$Particles.SpriteSize, B=<0.4 0.6>)) {isSetParameters=true}
  ]
  particleUpdateScript [
   &ParticleState(Kill Particles When Lifetime Has Elapsed=false, Let Infinitely Lived Particles Die When Emitter Deactivates=true)
   &ScaleSpriteSize(Scale Sprite Size Mode=::NewEnumerator0, Uniform Scale Factor=2)
   &ScaleColor(Scale RGB=&VectorFromFloat(Value=0.08), ScaleA=true, ScaleRGB=true)
   /SetVariables(
    Particles.EnergyColor:LinearColor=&Multiply_LinearColorByFloat(Linear Color=$Particles.Color, Float=10)
    Particles.EnergyDP:Vector4f=&MakeVector4(
     X=&FloatFromCurve(FloatCurve=DataInterface() {propertyValues {Curve {keys [<at 0 1>,<at 1 1>]}}}, Scale Curve=8)
     Y=0
     Z=0
     W=0
    )
   ) {isSetParameters=true}
   &FN_Light_Attributes(Color=$Particles.Color, Create Color Attribute=true, Create Radius Attribute=true, Radius=600, Scale Color=50)
  ]
  rendererClasses [&NiagaraSpriteRendererProperties, &NiagaraLightRendererProperties]
  renderers [
   {rendererClass=&NiagaraSpriteRendererProperties}
   {rendererClass=&NiagaraLightRendererProperties, rendererIndex=1}
   {rendererClass=&NiagaraSpriteRendererProperties, rendererIndex=2}
  ]
 }}
summary {
 systemName=NS_MaskMaker_SpireShards
 emitters [{emitterName=Flash_Flare, rendererClasses [&NiagaraSpriteRendererProperties, &NiagaraLightRendererProperties]}]
 userVariables []
}
