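# Niagara system for a spire death effect, named NS_Spire_Death, in the shared Spire VFX assets.
# Its system stack uses SystemState and two variable-assignment modules, with the FN_Fit math dynamic input as a dependency.
# User parameters expose primary and secondary colors, a ground-hit flag and a near-camera fade toggle.
# The summary lists no emitters or data interfaces, so no particle behavior is visible here.
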NiagaraSystem /SpireSharedAssets/VFX/NS_Spire_Death
dependencies {
 usedDataInterfaces []
 usedDynamicInputs [/Game/Effects/Niagara/DynamicInputs/Math/FN_Fit]
 usedModules [
  &self:SystemSpawnScript.SystemScriptSource.SystemScriptGraph.NiagaraNodeAssignment_4.SetVariables_#1
  &self:SystemSpawnScript.SystemScriptSource.SystemScriptGraph.NiagaraNodeAssignment_5.SetVariables_#2
  &SystemState
 ]
}
summary {
 systemName=NS_Spire_Death
 emitters []
 userVariables [
  User.IsHitGround {type=NiagaraBool, defaultValue=true}
  User.PrimaryColor {type=LinearColor, description="", defaultValue {struct=LinearColor, value=<rgba 5 0.2882 0.6068 1>}}
  User.SecondaryColor {type=LinearColor, description="", defaultValue {struct=LinearColor, value=<rgba 0.7101 0.0959 0.1594 5>}}
  User.UseNearCameraFade {type=NiagaraBool}
 ]
}
